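; BP_Holiday_OneGridPumpkins_A is a holiday pumpkin prop Blueprint for the Helios environment.
; Its construction script calls the parent's, then pushes the Width and Bend Power variables into the "Width" and "BendPower" material scalar parameters on StaticMesh1 and StaticMesh2.
; Shape tuning is therefore left to the meshes' materials, and the BeginPlay, overlap and Tick events are empty.
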(event EventBeginPlay)

(event Collision|EventActorBeginOverlap
  (params (OtherActor :type Actor%20Object%20Reference)))

(event EventTick
  (params (DeltaSeconds :type Float%20%28single-precision%29)))

(fn ConstructionScript
  (bind _staticmesh2 (Variables|Default|GetStaticMesh2))
  (bind _width (Variables|Default|GetWidth))
  (bind _bend_power (Variables|Default|GetBend%20Power))
  (|CallParentFunction)
  (Rendering|Material|SetScalarParameterValueonMaterials (Variables|Default|GetStaticMesh1) "Width" _width)
  (Rendering|Material|SetScalarParameterValueonMaterials _staticmesh2 "Width" _width)
  (Rendering|Material|SetScalarParameterValueonMaterials (Variables|Default|GetStaticMesh1) "BendPower" _bend_power)
  (Rendering|Material|SetScalarParameterValueonMaterials _staticmesh2 "BendPower" _bend_power))
