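; Blueprint for a Holiday one-grid pumpkins environment prop (variant C) in the Helios environment set.
; Its construction script runs the parent's logic, then copies the Width and BendPower variables into same-named scalar material parameters on two static meshes.
; This leaves any resulting shape variation to the materials, set at construction time.
; BeginPlay, overlap and Tick events are empty, so the Blueprint adds no runtime behavior of its own.
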(event EventBeginPlay)

(event Collision|EventActorBeginOverlap
  (params (OtherActor :type Actor%20Object%20Reference)))

(event EventTick
  (params (DeltaSeconds :type Float%20%28single-precision%29)))

(fn ConstructionScript
  (bind _staticmesh2 (Variables|Default|GetStaticMesh2))
  (bind _staticmesh1 (Variables|Default|GetStaticMesh1))
  (bind _width (Variables|Default|GetWidth))
  (bind _bendpower (Variables|Default|GetBendPower))
  (|CallParentFunction)
  (Rendering|Material|SetScalarParameterValueonMaterials _staticmesh1 "Width" _width)
  (Rendering|Material|SetScalarParameterValueonMaterials _staticmesh2 "Width" _width)
  (Rendering|Material|SetScalarParameterValueonMaterials _staticmesh1 "BendPower" _bendpower)
  (Rendering|Material|SetScalarParameterValueonMaterials _staticmesh2 "BendPower" _bendpower))
