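; BP_Holiday_OneGridPumpkins_E is a Blueprint for a holiday pumpkin environment prop, one variant of the OneGridPumpkins_A set.
; Its Construction Script calls the parent version, then writes the Width variable into the "Width" scalar material parameter on StaticMesh1 and StaticMesh2.
; BeginPlay, overlap and Tick events are empty, so this Blueprint adds no runtime logic of its own.
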(event EventBeginPlay)

(event Collision|EventActorBeginOverlap
  (params (OtherActor :type Actor%20Object%20Reference)))

(event EventTick
  (params (DeltaSeconds :type Float%20%28single-precision%29)))

(fn ConstructionScript
  (bind _width (Variables|Default|GetWidth))
  (|CallParentFunction)
  (Rendering|Material|SetScalarParameterValueonMaterials (Variables|Default|GetStaticMesh1) "Width" _width)
  (Rendering|Material|SetScalarParameterValueonMaterials (Variables|Default|GetStaticMesh2) "Width" _width))
